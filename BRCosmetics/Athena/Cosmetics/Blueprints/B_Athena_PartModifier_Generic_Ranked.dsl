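; Generic cosmetic part modifier that restyles a ranked character part, evidently a backpack, to match player rank.
; A watched rank stat or debug override feeds SetCurrentProgress, which sets a tiered dynamic material (Bronze to Unreal) and idle VFX from Elite upward.
; A chosen rank variant is capped at current progress unless a CVar, frontend display or Warmup phase allows it.
; SetRank previews ranks in editor or PIE.
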(event Custom|SetRank
  (params (Rank :type Integer))
  ;[0.4121 0.08369 1] This is a ke command to debug the asset rank in runtime. Just type: "ke * SetRank 0", being "0" the rank you want to debug.
  ;;[1 0.9115 0.5051] Only run if in Editor or Pie
  (if (Editor|InEditororPIE)
    (Default|BackpackMaterialSetup Rank)
    (CallFunction|VFXUpdate Rank)))

(event EventOnPawnAndPartChanged
  (bind _returnvalue (CustomCharacter|GetAssociatedPlayerPawn))
  (|CallParentFunction)
  (if (Variables|Rank|Debug|GetDEBUG_RankedProgression)
    (Default|SetCurrentProgress (Variables|Rank|Debug|GetDEBUG_RankedProgressionLevel))
    (else
      (Utilities|IsValid _returnvalue
        (:Is%20Valid
          (bind (_statname _statvalue) (Cosmetics|WatchCosmeticStat _returnvalue (Variables|Rank|GetRank%20Stat%20Tag))
            (:OnStatChanged
              (Default|SetCurrentProgress _statvalue))))))))

(event EventOnCosmeticPartAppliedPostResIn
  (params (Part :type Custom%20Character%20Part%20Object%20Reference))
  (|CallParentFunction Part)
  (Default|SetCurrentProgress (Variables|Rank|GetCurrentProgressionLevel)))

(event EventOnPostResIn
  (|CallParentFunction)
  (Default|SetCurrentProgress (Variables|Rank|GetCurrentProgressionLevel)))

(fn BackpackMaterialSetup
  (params (Index :type Integer))
  (bind _returnvalue (Rendering|Material|CreateDynamicMaterialInstance (Utilities|Select :Index Index :Option%200 (Variables|Rank|RankMaterials|GetBronze%20Material) :Option%201 (Variables|Rank|RankMaterials|GetBronze%20Material) :Option%202 (Variables|Rank|RankMaterials|GetBronze%20Material) :Option%203 (Variables|Rank|RankMaterials|GetBronze%20Material) :Option%204 (Variables|Rank|RankMaterials|GetSilver%20Material) :Option%205 (Variables|Rank|RankMaterials|GetSilver%20Material) :Option%206 (Variables|Rank|RankMaterials|GetSilver%20Material) :Option%207 (Variables|Rank|RankMaterials|GetGold%20Material) :Option%208 (Variables|Rank|RankMaterials|GetGold%20Material) :Option%209 (Variables|Rank|RankMaterials|GetGold%20Material) :Option%2010 (Variables|Rank|RankMaterials|GetPlatnium%20Material) :Option%2011 (Variables|Rank|RankMaterials|GetPlatnium%20Material) :Option%2012 (Variables|Rank|RankMaterials|GetPlatnium%20Material) :Option%2013 (Variables|Rank|RankMaterials|GetDiamond%20Material) :Option%2014 (Variables|Rank|RankMaterials|GetDiamond%20Material) :Option%2015 (Variables|Rank|RankMaterials|GetDiamond%20Material) :Option%2016 (Variables|Rank|RankMaterials|GetElite%20Material) :Option%2017 (Variables|Rank|RankMaterials|GetElite%20Material) :Option%2018 (Variables|Rank|RankMaterials|GetElite%20Material) :Option%2019 (Variables|Rank|RankMaterials|GetChampion%20Material) :Option%2020 (Variables|Rank|RankMaterials|GetChampion%20Material) :Option%2021 (Variables|Rank|RankMaterials|GetChampion%20Material) :Option%2022 (Variables|Rank|RankMaterials|GetUnreal%20Material)) "None" "None" :!class /Script/Engine.KismetMaterialLibrary))
  (Rendering|Material|SetMaterial (CustomCharacter|GetSkeletalMeshForPartOnPlayerPawn) 0 _returnvalue))

(fn GetChosenHabaneroRank
  (bind _asfort_player_pawn (Utilities|Casting|CastToFortPlayerPawn (Actor|GetOwner)))
  (bind (_outvariantchannels _bfoundchannel) (Class|FortCosmeticFunctionLibrary|%28Deprecated%29GetVariantChannelInfoForSlotTag (CustomCharacter|GetLoadoutSlotTagforPartType (CustomCharacter|GetCharacterPartType)) (Cosmetics|GetPawnCosmeticLoadout _asfort_player_pawn false true) "(TagName=\"Cosmetics.Variant.Channel.Ranked\")" false))
  (return _bfoundchannel (Class|FortCosmeticRankedVariant|GetHabaneroRankfromRankIndex (Utilities|String|StringToInteger (.CustomData (Utilities|Struct|BreakMcpVariantChannelInfo _outvariantchannels))))))

(fn SetCurrentProgress
  (params (NewProgress :type Integer))
  (bind _returnvalue (CallFunction|ShouldDisplayChosenRank))
  (bind _chosenvarianthabanerorank (LocalVariables|GetChosenVariantHabaneroRank))
  (bind _chosenvarianthabanerorank_2 (LocalVariables|GetChosenVariantHabaneroRank))
  (bind _isusingvariants (LocalVariables|GetIsUsingVariants))
  (bind _currentprogressionlevel (Variables|Rank|GetCurrentProgressionLevel))
  (bind (_is_using_variants _chosen_habanero_rank) (CallFunction|GetChosenHabaneroRank))
  (LocalVariables|SetIsUsingVariants _is_using_variants)
  (LocalVariables|SetChosenVariantHabaneroRank _chosen_habanero_rank)
  (if (or (LocalVariables|GetIsUsingVariants) (!= (Variables|GetNewProgress) (Variables|Rank|GetCurrentProgressionLevel)))
    (bind _output_get (Variables|Rank|SetCurrentProgressionLevel (Variables|GetNewProgress)))
    (Default|BackpackMaterialSetup (Math|Integer|SelectInt (Math|Integer|SelectInt _chosenvarianthabanerorank_2 (Math|Integer|Min(Integer) _output_get _chosenvarianthabanerorank) _returnvalue) _currentprogressionlevel _isusingvariants))
    (CallFunction|VFXUpdate (Math|Integer|SelectInt (Math|Integer|SelectInt _chosenvarianthabanerorank_2 (Math|Integer|Min(Integer) _output_get _chosenvarianthabanerorank) _returnvalue) _currentprogressionlevel _isusingvariants))))

(fn ShouldDisplayChosenRank
  (bind _returnvalue (Development|GetConsoleVariableBoolValue "Fort.Cosmetics.EnableRankedVariants"))
  (if _returnvalue
    (return true)
    (elif (Reactive|Frontend|IsCosmeticObjectDisplayedinFrontendContext 0)
      (return true)
      (else
        (bind _output_get (LocalVariables|SetGamePhaseLogic (BattleRoyaleGamePhaseLogic|GetBattleRoyaleGamePhaseLogicComponent)))
        (Utilities|IsValid _output_get
          (:Is%20Valid
            (if (BattleRoyaleGamePhaseLogic|ShouldCosmeticUseGamePhaseBehaviour (LocalVariables|GetGamePhaseLogic))
              (return (Utilities|Enum|Equal(Enum) (BattleRoyaleGamePhaseLogic|GetGamePhase (LocalVariables|GetGamePhaseLogic)) "Warmup"))
              (else
                (return false))))
          (:Is%20Not%20Valid
            (return false)))))))

(fn ConstructionScript
  (|CallParentFunction))

(fn VFXUpdate
  (params (Progression%20Level :type Integer) (KeepIdleAlwaysOn :type Boolean))
  (bind _returnvalue (CallFunction|GetIdleVFXComponent))
  (if (< Progression%20Level 16)
    (if KeepIdleAlwaysOn
      (Effects|Components|ParticleSystem|SetIntParameter _returnvalue "RankedProgression" -1)
      (else
        (Components|Activation|Deactivate _returnvalue)))
    (else
      (Components|Activation|Activate _returnvalue false)
      (Effects|Components|ParticleSystem|SetIntParameter _returnvalue "RankedProgression" (Utilities|Select :Index Progression%20Level :Option%200 -1 :Option%201 -1 :Option%202 -1 :Option%203 -1 :Option%204 -1 :Option%205 -1 :Option%206 -1 :Option%207 -1 :Option%208 -1 :Option%209 -1 :Option%2010 -1 :Option%2011 -1 :Option%2012 -1 :Option%2013 -1 :Option%2014 -1 :Option%2015 -1 :Option%2016 0 :Option%2017 0 :Option%2018 0 :Option%2019 1 :Option%2020 1 :Option%2021 1 :Option%2022 2)))))
