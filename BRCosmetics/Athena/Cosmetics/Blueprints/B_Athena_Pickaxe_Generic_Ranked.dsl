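; Generic ranked pickaxe cosmetic whose mesh material and VFX rank follow the player's rank progress.
; On init or visibility change it watches the rank stat on the pawn, with a frontend watch as init fallback, feeding SetCurrentProgress.
; A chosen rank variant shows fully in frontend or qualifying Warmup, otherwise capped at current progress.
; Rank index selects Bronze through Unreal materials; debug variables and an editor-only SetRank event allow previews.
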(event Custom|SetRank
  (params (Rank :type Integer))
  ;[0.4121 0.08369 1] This is a ke command to debug the asset rank in runtime. Just type: "ke * SetRank 0", being "0" the rank you want to debug.
  (bind _output_get (Variables|Default|SetDebugRank Rank))
  ;;[1 0.9115 0.5051] Only run if in Editor or Pie
  (if (Editor|InEditororPIE)
    (Default|BackpackMaterialSetup _output_get)
    (Variables|Default|SetVFX_Rank _output_get)))

(event EventOnWeaponVisibilityChanged
  (params (bVisible :type Boolean) (bSetForLocalControllerOnly :type Boolean))
  (|CallParentFunction bVisible bSetForLocalControllerOnly)
  (if (and bVisible (Variables|Rank|Debug|GetDEBUG_RankedProgression))
    (Default|SetCurrentProgress (Variables|Rank|Debug|GetDEBUG_RankedProgressionLevel))
    (else
      (bind _asfort_player_pawn (Utilities|Casting|CastToFortPlayerPawn (Variables|Actor|GetInstigator)))
      (if (not (Variables|Rank|GetIsWatchingProgressStat))
        (Utilities|IsValid _asfort_player_pawn
          (:Is%20Valid
            (bind (_statname _statvalue) (Cosmetics|WatchCosmeticStat _asfort_player_pawn (Variables|Rank|GetRank%20Stat%20Tag))
              (:then
                (Variables|Rank|SetIsWatchingProgressStat true))
              (:OnStatChanged
                (Default|SetCurrentProgress _statvalue)))))))))

(event EventOnInitWeaponCosmetics
  (|CallParentFunction)
  (bind _asfort_player_pawn (Utilities|Casting|CastToFortPlayerPawn (Variables|Actor|GetInstigator))
    (:then
      (if (Variables|Rank|GetIsWatchingProgressStat)
        (Default|SetCurrentProgress (Variables|Rank|GetCurrentProgressionLevel))
        (else
          (Utilities|IsValid _asfort_player_pawn
            (:Is%20Valid
              (bind (_statname _statvalue) (Cosmetics|WatchCosmeticStat _asfort_player_pawn (Variables|Rank|GetRank%20Stat%20Tag))
                (:then
                  (Variables|Rank|SetIsWatchingProgressStat true))
                (:OnStatChanged
                  (Default|SetCurrentProgress _statvalue))))))))
    (:CastFailed
      (bind (_statname_2 _statvalue_2) (Cosmetics|WatchFrontendCosmeticStat(DONOTUSE) (Variables|Rank|GetRank%20Stat%20Tag))
        (:OnStatChanged
          (Default|SetCurrentProgress _statvalue_2))) ; Used as a workaround for the Ranked Stuff only
    )))

(fn BackpackMaterialSetup
  (params (Index :type Integer))
  (Rendering|Material|SetMaterial (Variables|Visual|GetWeaponMesh) 0 (Utilities|Select :Index Index :Option%200 (Variables|Rank|RankMaterials|GetBronze%20Material) :Option%201 (Variables|Rank|RankMaterials|GetBronze%20Material) :Option%202 (Variables|Rank|RankMaterials|GetBronze%20Material) :Option%203 (Variables|Rank|RankMaterials|GetBronze%20Material) :Option%204 (Variables|Rank|RankMaterials|GetSilver%20Material) :Option%205 (Variables|Rank|RankMaterials|GetSilver%20Material) :Option%206 (Variables|Rank|RankMaterials|GetSilver%20Material) :Option%207 (Variables|Rank|RankMaterials|GetGold%20Material) :Option%208 (Variables|Rank|RankMaterials|GetGold%20Material) :Option%209 (Variables|Rank|RankMaterials|GetGold%20Material) :Option%2010 (Variables|Rank|RankMaterials|GetPlatnium%20Material) :Option%2011 (Variables|Rank|RankMaterials|GetPlatnium%20Material) :Option%2012 (Variables|Rank|RankMaterials|GetPlatnium%20Material) :Option%2013 (Variables|Rank|RankMaterials|GetDiamond%20Material) :Option%2014 (Variables|Rank|RankMaterials|GetDiamond%20Material) :Option%2015 (Variables|Rank|RankMaterials|GetDiamond%20Material) :Option%2016 (Variables|Rank|RankMaterials|GetElite%20Material) :Option%2017 (Variables|Rank|RankMaterials|GetElite%20Material) :Option%2018 (Variables|Rank|RankMaterials|GetElite%20Material) :Option%2019 (Variables|Rank|RankMaterials|GetChampion%20Material) :Option%2020 (Variables|Rank|RankMaterials|GetChampion%20Material) :Option%2021 (Variables|Rank|RankMaterials|GetChampion%20Material) :Option%2022 (Variables|Rank|RankMaterials|GetUnreal%20Material))))

(fn GetChosenHabaneroRank
  (bind (_outvariantchannels _bfoundchannel) (Class|FortCosmeticFunctionLibrary|%28Deprecated%29GetVariantChannelInfoForSlotTag "(TagName=\"LoadoutSlot.Pickaxe\")" (CallFunction|GetCosmeticData) "(TagName=\"Cosmetics.Variant.Channel.Ranked\")" false))
  (return _bfoundchannel (Class|FortCosmeticRankedVariant|GetHabaneroRankfromRankIndex (Utilities|String|StringToInteger (.CustomData (Utilities|Struct|BreakMcpVariantChannelInfo _outvariantchannels))))))

(fn SetCurrentProgress
  (params (NewProgress :type Integer))
  (bind _returnvalue (CallFunction|ShouldDisplayChosenRank))
  (bind _chosenvarianthabanerorank (Variables|Default|GetChosenVariantHabaneroRank))
  (bind _chosenvarianthabanerorank_2 (Variables|Default|GetChosenVariantHabaneroRank))
  (bind _isusingvariants (Variables|Default|GetIsUsingVariants))
  (bind _currentprogressionlevel (Variables|Rank|GetCurrentProgressionLevel))
  (bind (_is_using_variants _chosen_habanero_rank) (CallFunction|GetChosenHabaneroRank))
  (Variables|Default|SetIsUsingVariants _is_using_variants)
  (Variables|Default|SetChosenVariantHabaneroRank _chosen_habanero_rank)
  (if (or (Variables|Default|GetIsUsingVariants) (!= (Variables|GetNewProgress) (Variables|Rank|GetCurrentProgressionLevel)))
    (bind _output_get (Variables|Rank|SetCurrentProgressionLevel (Variables|GetNewProgress)))
    (Default|BackpackMaterialSetup (Math|Integer|SelectInt (Math|Integer|SelectInt _chosenvarianthabanerorank_2 (Math|Integer|Min(Integer) _output_get _chosenvarianthabanerorank) _returnvalue) _currentprogressionlevel _isusingvariants))
    (Variables|Default|SetVFX_Rank (Math|Integer|SelectInt (Math|Integer|SelectInt _chosenvarianthabanerorank_2 (Math|Integer|Min(Integer) _output_get _chosenvarianthabanerorank) _returnvalue) _currentprogressionlevel _isusingvariants))))

(fn ShouldDisplayChosenRank
  (if (Reactive|Frontend|IsCosmeticObjectDisplayedinFrontendContext 0)
    (return true)
    (else
      (bind _output_get (LocalVariables|SetGamePhaseLogic (BattleRoyaleGamePhaseLogic|GetBattleRoyaleGamePhaseLogicComponent)))
      (Utilities|IsValid _output_get
        (:Is%20Valid
          (if (BattleRoyaleGamePhaseLogic|ShouldCosmeticUseGamePhaseBehaviour (LocalVariables|GetGamePhaseLogic))
            (return (Utilities|Enum|Equal(Enum) (BattleRoyaleGamePhaseLogic|GetGamePhase (LocalVariables|GetGamePhaseLogic)) "Warmup"))
            (else
              (return false))))
        (:Is%20Not%20Valid
          (return false))))))

(fn ConstructionScript
  (|CallParentFunction))
