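; Generic audio-reactive cosmetic part modifier for Athena that acts as a thin subclass of a shared part modifier.
; Its construction script only calls the parent implementation.
; Audio reactivity logic lives in the base-most B_Athena_PartModifier_Generic, gated by a Has Audio Reactivity boolean that this class sets to true.
; An in-graph note forbids adding support here to avoid divergence from other modifiers.
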(fn ConstructionScript
  ;[0.6 0 0 0.5] DO NOT ADD BLUEPRINT SUPPORT HERE!!!
  ; Audio Reactivity support goes in the base-most class B_Athena_PartModifier_Generic, where it can be controlled via the Has Audio Reactivity boolean, which is set to true in this class but is false there.
  ; If that class has changed, search for the appropriate class, but don't add support here where
  ; it's likely to cause divergence with other mutators.
  ; If you have any questions, please ask the Cosmetics Platform team.
  (|CallParentFunction))
